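# Masked, skeletal-mesh material for the Tred vehicle wheel, adding a paintable color overlay to the shared simple wheel material function.
# Nested lerps, gated by a mask texture (black by default) and inverted vertex-color green, blend base color toward OP_Color and also adjust metallic and roughness.
# The modified material attributes pass through the shared vehicle damage function, then the de-rez transition function.
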Material /VehicleCosmetics/Wheels/Tred/Materials/M_Wheel_Tred
graph {
 properties {
  bAutomaticallySetUsageInEditor=false
  bUseMaterialAttributes=true
  bUsedWithMeshDeformer=true
  bUsedWithSkeletalMesh=true
  blendMode=BLEND_Masked
  pixelDepthOffsetMode=PDOM_Legacy
  usedShadingModels=Default Lit
 }
 nodes [
  n1 = GetMaterialAttributes(None=MaterialFunctionCall() {materialFunction=&MF_WheelParent_Simple}.MaterialAttributes) {attributeGetTypes [#1, #2, #3]}
  n2 = LinearInterpolate(
   A=@n1.Base Color
   B=LinearInterpolate(A=@n1.Base Color, B=VectorParameter() {parameterName=OP_Color, defaultValue=<rgba 1 0 0 1>}.RGB, Alpha=NamedRerouteUsage())
   Alpha=NamedRerouteUsage()
  )
  n3 = NamedRerouteUsage()
  n4 = NamedRerouteUsage()
  n5 = LinearInterpolate(A=@n1.Metallic, B=LinearInterpolate(A=@n1.Metallic, Alpha=@n3), Alpha=@n4)
  n6 = LinearInterpolate(A=@n1.Roughness, B=LinearInterpolate(A=@n1.Roughness, Alpha=@n3), Alpha=@n4)
  MaterialFunctionCall(In=MaterialFunctionCall(Material=SetMaterialAttributes!(MaterialAttributes=@n1.MaterialAttributes, Base Color=@n2, Metallic=@n5, Roughness=@n6) {attributeSetTypes [#1, #2, #3]}) {materialFunction=&MF_BaseVehicle_Damage}.Result) {materialFunction=&MF_VehicleDeRez_Transition}
  NamedRerouteDeclaration(None=TextureSampleParameter2D() {parameterName=OP_mask, texture=/VehicleCosmetics/Textures/T_Black}.RGB) {name=mask, nodeColor=<rgba 1 0 0.1406 1>}
  NamedRerouteDeclaration(None=OneMinus(None=VertexColor().G)) {name=Vert, nodeColor=<rgba 0.3828 1 0 1>}
 ]
}
